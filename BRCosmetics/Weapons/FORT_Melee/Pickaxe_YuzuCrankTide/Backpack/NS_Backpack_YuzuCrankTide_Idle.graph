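# Niagara idle glow for the YuzuCrankTide pickaxe on the back: persistent burst sprites sized by owner scale.
# Idle_Glow is a soft red sprite; Reactive_Glow layers sprites and meshes in orange-red, with time-of-day base color.
# A FadeInFadeOut on user timestamps, times ReactivityOn, drives the reactive colors and alpha, dimmed in the front end.
# User.ReactiveBool zeroes the scale factors on both emitters when true.
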NiagaraSystem /BRCosmetics/Weapons/FORT_Melee/Pickaxe_YuzuCrankTide/Backpack/NS_Backpack_YuzuCrankTide_Idle
dependencies {usedDataInterfaces [&NiagaraDataInterfaceCurve], usedDynamicInputs [], usedModules [&SystemState]}
emitters {
 Idle_Glow {
  emitterUpdateScript [
   &EmitterState()
   &SpawnBurst_Instantaneous(Use Spawn Probability=false, UseLoopCountLimit=false)
  ]
  particleSpawnScript [
   &InitializeParticle(
    Color=<rgba 0.9062 0.0943 0.0661 0.5>
    Lifetime=2
    Position Offset Coordinate Space=::Local
    Sprite Size=&Multiply_Vector2D(
     A=<38 38>
     B=&MakeVector2D(
      X=&MakeFloatFromVector(Vector=$Engine.Owner.Scale, Channel=::NewEnumerator0)
      Y=&MakeFloatFromVector(Vector=$Engine.Owner.Scale, Channel=::NewEnumerator2)
     )
    )
    Sprite Size Mode=::NewEnumerator4
    UsePositionOffset=true
    Write Material Random=true
   )
   &SpriteFacingAndAlignment(Alignment Coordinate Space=::Local, Sprite Alignment=<0 0 -1>, Sprite Facing=<0 1 0>, WriteAlignment=true, WriteFacing=true)
  ]
  particleUpdateScript [
   &ParticleState(Kill Particles When Lifetime Has Elapsed=false, Let Infinitely Lived Particles Die When Emitter Deactivates=true)
   &ScaleColor(
    Scale Alpha=&FloatFromCurve(
     FloatCurve=DataInterface() {propertyValues {Curve {defaultValue=1, keys []}}}
     Scale Curve=&MakeCustomFloatFromBool(Bool=$User.ReactiveBool, True Float=0, False Float=1)
    )
    ScaleA=true
    ScaleRGB=true
   )
   &ScaleColor(Scale Alpha=0.4, ScaleA=true, ScaleRGB=true) {module=ScaleColor001}
  ]
  rendererClasses [&NiagaraSpriteRendererProperties]
  renderers [{rendererClass=&NiagaraSpriteRendererProperties}]
 }
 Reactive_Glow {
  emitterUpdateScript [
   &EmitterState()
   &SpawnBurst_Instantaneous(Use Spawn Probability=false, UseLoopCountLimit=false)
  ]
  particleSpawnScript [
   &InitializeParticle(
    Color=&FNTimeOfDayColor(
     Night Color=<rgba 1.9375 0.1549 0.1734 0.05>
     Morning Color=<rgba 1.9375 0.1025 0.0395 0.05>
     Day Color=<rgba 1 0.0239 0 0.15>
     Evening Color=<rgba 1.9375 0.1025 0.0395 0.05>
    )
    Lifetime=2
    Position Offset=<0 -0.25 0>
    Position Offset Coordinate Space=::Local
    Sprite Size=&Multiply_Vector2D(
     A=<34.5 34.5>
     B=&MakeVector2D(
      X=&MakeFloatFromVector(Vector=$Engine.Owner.Scale, Channel=::NewEnumerator0)
      Y=&MakeFloatFromVector(Vector=$Engine.Owner.Scale, Channel=::NewEnumerator2)
     )
    )
    Sprite Size Mode=::NewEnumerator4
    UsePositionOffset=true
    Write Material Random=true
   )
   &SpriteFacingAndAlignment(Alignment Coordinate Space=::Local, Sprite Alignment=<0 0 -1>, Sprite Facing=<0 1 0>, WriteAlignment=true, WriteFacing=true)
   &CameraOffset(Camera Offset Amount=2.2)
   &InitialMeshOrientation(Mesh Axis To Orient=<0 0 -1>, Mesh Orientation Mode=::NewEnumerator2, Orientation Vector=<0 1 0>, Use Rotation Vector=true)
   /SetVariables(Particles.TightGlowColorScale:Vector3f=&Multiply_Vector(A=$Particles.Scale, B=<1 1 1>)) {isSetParameters=true}
   /SetVariables(Particles.MeshCameraOffset=2) {isSetParameters=true}
   /SetVariables(Particles.MeshScale:Vector3f=&Multiply_Vector(A=&Multiply_Vector(A=<0.63 0.63 2>, B=<0.92 0.92 0.9>), B=<1 1 1>)) {isSetParameters=true}
  ]
  particleUpdateScript [
   &ParticleState(Kill Particles When Lifetime Has Elapsed=false, Let Infinitely Lived Particles Die When Emitter Deactivates=true)
   /SetVariables(
    Particles.Orange:LinearColor=&Multiply_LinearColorByFloat(
     Linear Color=<rgba 2 0.6348 0 1>
     Float=&Multiply_Float(
      A=&Multiply_Float(
       A=&Multiply_Float(
        A=&self:FadeInFadeOut(
         Time:NiagaraFloat=$Engine.Time
         Timestamp_0:NiagaraFloat=$User.Timestamp_0
         KillTransitionLength:NiagaraFloat=1.5
         KillAnimationLength:NiagaraFloat=2
         Timestamp_1:NiagaraFloat=0
        )
        B=$User.ReactivityOn
       )
       B=&MakeCustomFloatFromBool(Bool=$User.ReactiveBool, True Float=0, False Float=1)
      )
      B=&FN_FrontEndFloatModifier(Gameplay=1, Front End=0.4)
     )
    )
    Particles.TightColor:LinearColor=&Multiply_LinearColorByFloat(
     Linear Color=<rgba 50 7.6515 1.8229 1>
     Float=&Multiply_Float(
      A=&Multiply_Float(
       A=&Multiply_Float(
        A=&self:FadeInFadeOut(
         Time:NiagaraFloat=$Engine.Time
         Timestamp_0:NiagaraFloat=$User.Timestamp_0
         KillTransitionLength:NiagaraFloat=1.5
         KillAnimationLength:NiagaraFloat=2
         Timestamp_1:NiagaraFloat=$User.Timestamp_1
        )
        B=$User.ReactivityOn
       )
       B=&MakeCustomFloatFromBool(Bool=$User.ReactiveBool, True Float=0, False Float=1)
      )
      B=&FN_FrontEndFloatModifier(Gameplay=1, Front End=0.5)
     )
    )
   ) {isSetParameters=true}
   &ScaleColor(
    Scale Alpha=&FloatFromCurve(
     FloatCurve=DataInterface() {propertyValues {Curve {defaultValue=1, keys []}}}
     Scale Curve=&Multiply_Float(
      A=&Multiply_Float(
       A=&self:FadeInFadeOut(
        Time:NiagaraFloat=$Engine.Time
        Timestamp_0:NiagaraFloat=$User.Timestamp_0
        KillTransitionLength:NiagaraFloat=1.5
        KillAnimationLength:NiagaraFloat=2
        Timestamp_1:NiagaraFloat=$User.Timestamp_1
       )
       B=$User.ReactivityOn
      )
      B=&MakeCustomFloatFromBool(Bool=$User.ReactiveBool, True Float=0, False Float=1)
     )
    )
    ScaleA=true
    ScaleRGB=true
   )
   /SetVariables(
    Particles.AmbientOrange:Vector2f=&Multiply_Vector2D(A=$Particles.SpriteSize, B=&Vector2DFromFloat(Value=1.05))
    Particles.AmbientSize:Vector2f=&Multiply_Vector2D(A=$Particles.SpriteSize, B=&Vector2DFromFloat(Value=1.3))
   ) {isSetParameters=true}
   /SetVariables(Particles.GlowColor=&Multiply_LinearColorByFloat(
     Linear Color=<rgba 0.5 0 0.0177 0.3>
     Float=&Multiply_Float(
      A=&Multiply_Float(
       A=&Multiply_Float(
        A=&self:FadeInFadeOut(
         Time:NiagaraFloat=$Engine.Time
         Timestamp_0:NiagaraFloat=$User.Timestamp_0
         KillTransitionLength:NiagaraFloat=1.5
         KillAnimationLength:NiagaraFloat=2
         Timestamp_1:NiagaraFloat=$User.Timestamp_1
        )
        B=$User.ReactivityOn
       )
       B=&MakeCustomFloatFromBool(Bool=$User.ReactiveBool, True Float=0, False Float=1)
      )
      B=&FN_FrontEndFloatModifier(Gameplay=1, Front End=0.5)
     )
    )) {isSetParameters=true}
   /SetVariables(Particles.TightGlowColor:LinearColor=&Multiply_LinearColorByFloat(
     Linear Color=<rgba 2 0.38 0.0645 1>
     Float=&Multiply_Float(
      A=&Multiply_Float(
       A=&Multiply_Float(
        A=&self:FadeInFadeOut(
         Time:NiagaraFloat=$Engine.Time
         Timestamp_0:NiagaraFloat=$User.Timestamp_0
         KillTransitionLength:NiagaraFloat=1.5
         KillAnimationLength:NiagaraFloat=2
         Timestamp_1:NiagaraFloat=$User.Timestamp_1
        )
        B=$User.ReactivityOn
       )
       B=&MakeCustomFloatFromBool(Bool=$User.ReactiveBool, True Float=0, False Float=1)
      )
      B=&FN_FrontEndFloatModifier(Gameplay=1, Front End=0.5)
     )
    )) {isSetParameters=true}
   &ScaleColor(Scale Alpha=&FN_FrontEndFloatModifier(Gameplay=1, Front End=0.3), ScaleA=true, ScaleRGB=true) {module=ScaleColor001}
  ]
  rendererClasses [&NiagaraSpriteRendererProperties, &NiagaraMeshRendererProperties]
  renderers [
   {rendererClass=&NiagaraSpriteRendererProperties}
   {rendererClass=&NiagaraSpriteRendererProperties, rendererIndex=1}
   {rendererClass=&NiagaraSpriteRendererProperties, rendererIndex=2}
   {rendererClass=&NiagaraMeshRendererProperties, rendererIndex=3}
   {rendererClass=&NiagaraMeshRendererProperties, rendererIndex=4}
  ]
 }
}
summary {
 systemName=NS_Backpack_YuzuCrankTide_Idle
 emitters [
  {emitterName=Idle_Glow, rendererClasses [&NiagaraSpriteRendererProperties]}
  {emitterName=Reactive_Glow, rendererClasses [&NiagaraSpriteRendererProperties, &NiagaraMeshRendererProperties]}
 ]
 userVariables [
  User.bIsFrontEnd {type=NiagaraBool}
  User.bIsFrontEndPreview {type=NiagaraBool}
  User.ReactiveBool {type=NiagaraBool}
  User.ReactivityOn {type=NiagaraFloat, defaultValue=1, description=""}
  User.Timestamp_0 {type=NiagaraFloat, defaultValue=-100, description=""}
  User.Timestamp_1 {type=NiagaraFloat, defaultValue=-100, description=""}
 ]
}
